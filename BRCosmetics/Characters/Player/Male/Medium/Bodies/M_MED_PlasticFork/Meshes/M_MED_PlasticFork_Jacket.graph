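# Physics asset for the jacket of the male medium PlasticFork character body, simulating the lower jacket around hips and thighs.
# Seven capsule chains (back, plus front, side and rear per side) hang from master bones locked to the pelvis, with tight twist and mostly one-axis swing.
# Angularly free cross-constraints link neighboring chain tips on each side, and the left front adds a small Lock body.
# Pelvis, lower spine and thigh capsules supply the body shapes.
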PhysicsAsset /BRCosmetics/Characters/Player/Male/Medium/Bodies/M_MED_PlasticFork/Meshes/M_MED_PlasticFork_Jacket
bodies {
 dyn_jacket_bk [{
   shapeType=Capsule
   length=6.4885
   radius=1.3768
   shapeName=dyn_jacket_bk_capsule
   center=<4.0877 0.1689 0.0284>
   rotation=<rot 0 0.4821 -89.3737>
  }]
 dyn_jacket_bk_01 [{shapeType=Capsule, length=3.5, radius=4.5, shapeName=dyn_jacket_bk_l_01_capsule, center=<5 0 3>, rotation=<rot 0 0 90>}]
 dyn_jacket_bk_MSTR [{shapeType=Capsule, length=0.505, radius=0.505, shapeName=dyn_jacket_bk_MSTR_capsule, rotation=<rot 0 -90 0>}]
 dyn_jacket_l_01 [{shapeType=Capsule, length=5.7236, radius=6.2764, shapeName=dyn_jacket_l_01_capsule, center=<4 0 6>, rotation=<rot 0 0 90>}]
 dyn_jacket_l_02 [{shapeType=Capsule, length=6.8924, radius=5.1076, shapeName=dyn_jacket_l_02_capsule, center=<4 0 6>, rotation=<rot 0 0 90>}]
 dyn_jacket_l_MSTR [{shapeType=Capsule, length=0.505, radius=0.505, shapeName=dyn_jacket_l_MSTR_capsule, rotation=<rot 0 -90 0>}]
 dyn_jacket_l_bk_01 [{shapeType=Capsule, length=8.6557, radius=5.3443, shapeName=dyn_jacket_l_bk_01_capsule, center=<4 0 5>, rotation=<rot 0 0 90>}]
 dyn_jacket_l_bk_02 [{shapeType=Capsule, length=9.1633, radius=4.8367, shapeName=dyn_jacket_l_bk_02_capsule, center=<4 0 5>, rotation=<rot 0 0 90>}]
 dyn_jacket_l_bk_MSTR [{shapeType=Capsule, length=0.505, radius=0.505, shapeName=dyn_jacket_l_bk_MSTR_capsule, rotation=<rot 0 -90 0>}]
 dyn_jacket_l_fr_01 [{shapeType=Capsule, length=7.8256, radius=4.4244, shapeName=dyn_jacket_l_fr_01_capsule, center=<5 -2 4.5>, rotation=<rot 0 0 84.9986>}]
 dyn_jacket_l_fr_02 [
  {
   shapeType=Capsule
   length=5.4811
   radius=6.7689
   shapeName=dyn_jacket_l_fr_02_capsule
   center=<5.4256 -1.9392 3.8052>
   rotation=<rot 0 0 84.9986>
  }
  {
   shapeType=Capsule
   length=2.4431
   radius=2.6019
   shapeName=dyn_jacket_l_fr_02_capsule
   center=<7.6711 5.0759 -1.0213>
   rotation=<rot -87.5564 0.0001 -0.0001>
  }
 ]
 dyn_jacket_l_fr_Lock [{
   shapeType=Capsule
   length=0.3414
   radius=2.3504
   shapeName=dyn_jacket_l_fr_Lock_capsule
   center=<4.1719 0.1624 0.0823>
   rotation=<rot 0 92.4512 -89.3014>
  }]
 dyn_jacket_l_fr_MSTR [{shapeType=Capsule, length=0.505, radius=0.505, shapeName=dyn_jacket_l_fr_MSTR_capsule, rotation=<rot 0 -90 0>}]
 dyn_jacket_r_01 [{shapeType=Capsule, length=6.625, radius=6.75, shapeName=dyn_jacket_l_01_capsule, center=<4 0 6>, rotation=<rot 0 0 90>}]
 dyn_jacket_r_02 [{shapeType=Capsule, length=7.375, radius=6, shapeName=dyn_jacket_l_02_capsule, center=<4 0 6>, rotation=<rot 0 0 90>}]
 dyn_jacket_r_MSTR [{shapeType=Capsule, length=0.505, radius=0.505, shapeName=dyn_jacket_r_MSTR_capsule, rotation=<rot 0 -90 0>}]
 dyn_jacket_r_bk_01 [{shapeType=Capsule, length=7.75, radius=6.25, shapeName=dyn_jacket_l_bk_01_capsule, center=<4 0 5>, rotation=<rot 0 0 90>}]
 dyn_jacket_r_bk_02 [{shapeType=Capsule, length=8, radius=6, shapeName=dyn_jacket_l_bk_02_capsule, center=<4 0 5>, rotation=<rot 0 0 90>}]
 dyn_jacket_r_bk_MSTR [{shapeType=Capsule, length=0.505, radius=0.505, shapeName=dyn_jacket_r_bk_MSTR_capsule, rotation=<rot 0 -90 0>}]
 dyn_jacket_r_fr_01 [{shapeType=Capsule, length=6.5, radius=5.75, shapeName=dyn_jacket_l_fr_01_capsule, center=<5 2 4.5>, rotation=<rot 0 0 95>}]
 dyn_jacket_r_fr_02 [{shapeType=Capsule, length=6.25, radius=6, shapeName=dyn_jacket_l_fr_02_capsule, center=<5 2 4.5>, rotation=<rot 0 0 95>}]
 dyn_jacket_r_fr_MSTR [{shapeType=Capsule, length=0.505, radius=0.505, shapeName=dyn_jacket_r_fr_MSTR_capsule, rotation=<rot 0 -90 0>}]
 pelvis [{shapeType=Capsule, length=7.8566, radius=10.8609, center=<-4.1085 -0.1339 0>}]
 spine_01 [{shapeType=Capsule, length=8.7202, radius=8.919, center=<-1.946 0.3523 0>}]
 thigh_twist_01_l [{shapeType=Capsule, length=28.005, radius=8.543, center=<8.0003 -0.8543 0.6184>, rotation=<rot 88.1882 5.0001 5.3096>}]
 thigh_twist_01_r [{shapeType=Capsule, length=26.9162, radius=8.543, center=<-8.0239 -0.032 -1.2174>, rotation=<rot 88.1882 5.0001 -174.6904>}]
}
constraints [
 {
  bone1Name=spine_01
  bone2Name=pelvis
  swing1LimitDegrees=45
  swing1Motion=Locked
  swing2LimitDegrees=45
  swing2Motion=Locked
  twistLimitDegrees=45
  twistMotion=Locked
 }
 {
  bone1Name=thigh_twist_01_r
  bone2Name=pelvis
  swing1LimitDegrees=45
  swing1Motion=Locked
  swing2LimitDegrees=45
  swing2Motion=Locked
  twistLimitDegrees=45
  twistMotion=Locked
 }
 {
  bone1Name=thigh_twist_01_l
  bone2Name=pelvis
  swing1LimitDegrees=45
  swing1Motion=Locked
  swing2LimitDegrees=45
  swing2Motion=Locked
  twistLimitDegrees=45
  twistMotion=Locked
 }
 {
  bone1Name=dyn_jacket_bk_MSTR
  bone2Name=pelvis
  swing1LimitDegrees=45
  swing1Motion=Locked
  swing2LimitDegrees=45
  swing2Motion=Locked
  twistLimitDegrees=45
  twistMotion=Locked
 }
 {bone1Name=dyn_jacket_bk, bone2Name=dyn_jacket_bk_MSTR, swing1LimitDegrees=30, swing2LimitDegrees=65, twistLimitDegrees=2}
 {bone1Name=dyn_jacket_bk_01, bone2Name=dyn_jacket_bk, swing1LimitDegrees=5, swing2LimitDegrees=30, twistLimitDegrees=2}
 {
  bone1Name=dyn_jacket_l_MSTR
  bone2Name=pelvis
  swing1LimitDegrees=45
  swing1Motion=Locked
  swing2LimitDegrees=45
  swing2Motion=Locked
  twistLimitDegrees=45
  twistMotion=Locked
 }
 {bone1Name=dyn_jacket_l_01, bone2Name=dyn_jacket_l_MSTR, swing1LimitDegrees=5, swing2LimitDegrees=30, twistLimitDegrees=2}
 {bone1Name=dyn_jacket_l_02, bone2Name=dyn_jacket_l_01, swing1LimitDegrees=5, swing2LimitDegrees=30, twistLimitDegrees=2}
 {
  bone1Name=dyn_jacket_l_bk_MSTR
  bone2Name=pelvis
  swing1LimitDegrees=45
  swing1Motion=Locked
  swing2LimitDegrees=45
  swing2Motion=Locked
  twistLimitDegrees=45
  twistMotion=Locked
 }
 {bone1Name=dyn_jacket_l_bk_01, bone2Name=dyn_jacket_l_bk_MSTR, swing1LimitDegrees=5, swing2LimitDegrees=30, twistLimitDegrees=2}
 {bone1Name=dyn_jacket_l_bk_02, bone2Name=dyn_jacket_l_bk_01, swing1LimitDegrees=5, swing2LimitDegrees=20, twistLimitDegrees=2}
 {
  bone1Name=dyn_jacket_l_fr_MSTR
  bone2Name=pelvis
  swing1LimitDegrees=45
  swing1Motion=Locked
  swing2LimitDegrees=45
  swing2Motion=Locked
  twistLimitDegrees=45
  twistMotion=Locked
 }
 {bone1Name=dyn_jacket_l_fr_01, bone2Name=dyn_jacket_l_fr_MSTR, swing1LimitDegrees=5, swing2LimitDegrees=30, twistLimitDegrees=2}
 {bone1Name=dyn_jacket_l_fr_02, bone2Name=dyn_jacket_l_fr_01, swing1LimitDegrees=5, swing2LimitDegrees=30, twistLimitDegrees=5}
 {
  bone1Name=dyn_jacket_r_MSTR
  bone2Name=pelvis
  swing1LimitDegrees=45
  swing1Motion=Locked
  swing2LimitDegrees=45
  swing2Motion=Locked
  twistLimitDegrees=45
  twistMotion=Locked
 }
 {bone1Name=dyn_jacket_r_01, bone2Name=dyn_jacket_r_MSTR, swing1LimitDegrees=5, swing2LimitDegrees=30, twistLimitDegrees=2}
 {bone1Name=dyn_jacket_r_02, bone2Name=dyn_jacket_r_01, swing1LimitDegrees=5, swing2LimitDegrees=30, twistLimitDegrees=2}
 {
  bone1Name=dyn_jacket_r_bk_MSTR
  bone2Name=pelvis
  swing1LimitDegrees=45
  swing1Motion=Locked
  swing2LimitDegrees=45
  swing2Motion=Locked
  twistLimitDegrees=45
  twistMotion=Locked
 }
 {bone1Name=dyn_jacket_r_bk_01, bone2Name=dyn_jacket_r_bk_MSTR, swing1LimitDegrees=5, swing2LimitDegrees=30, twistLimitDegrees=2}
 {bone1Name=dyn_jacket_r_bk_02, bone2Name=dyn_jacket_r_bk_01, swing1LimitDegrees=5, swing2LimitDegrees=30, twistLimitDegrees=2}
 {
  bone1Name=dyn_jacket_r_fr_MSTR
  bone2Name=pelvis
  swing1LimitDegrees=45
  swing1Motion=Locked
  swing2LimitDegrees=45
  swing2Motion=Locked
  twistLimitDegrees=45
  twistMotion=Locked
 }
 {bone1Name=dyn_jacket_r_fr_01, bone2Name=dyn_jacket_r_fr_MSTR, swing1LimitDegrees=5, swing2LimitDegrees=30, twistLimitDegrees=2}
 {bone1Name=dyn_jacket_r_fr_02, bone2Name=dyn_jacket_r_fr_01, swing1LimitDegrees=5, swing2LimitDegrees=30, twistLimitDegrees=5}
 {
  bone1Name=dyn_jacket_r_02
  bone2Name=dyn_jacket_r_fr_02
  swing1LimitDegrees=45
  swing1Motion=Free
  swing2LimitDegrees=45
  swing2Motion=Free
  twistLimitDegrees=45
  twistMotion=Free
 }
 {
  bone1Name=dyn_jacket_r_bk_02
  bone2Name=dyn_jacket_r_02
  swing1LimitDegrees=45
  swing1Motion=Free
  swing2LimitDegrees=45
  swing2Motion=Free
  twistLimitDegrees=45
  twistMotion=Free
 }
 {
  bone1Name=dyn_jacket_l_02
  bone2Name=dyn_jacket_l_fr_02
  swing1LimitDegrees=45
  swing1Motion=Free
  swing2LimitDegrees=45
  swing2Motion=Free
  twistLimitDegrees=45
  twistMotion=Free
 }
 {
  bone1Name=dyn_jacket_l_bk_02
  bone2Name=dyn_jacket_l_02
  swing1LimitDegrees=45
  swing1Motion=Free
  swing2LimitDegrees=45
  swing2Motion=Free
  twistLimitDegrees=45
  twistMotion=Free
 }
 {bone1Name=dyn_jacket_l_fr_Lock, bone2Name=dyn_jacket_l_fr_02, swing1LimitDegrees=0, swing2LimitDegrees=30, twistLimitDegrees=0}
]
