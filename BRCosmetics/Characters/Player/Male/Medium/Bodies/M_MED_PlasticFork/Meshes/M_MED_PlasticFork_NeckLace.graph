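# Physics asset for the necklace and collar of the medium male PlasticFork player body, built entirely from capsules.
# Spine, clavicle and necklace master bodies are locked together by constraints, forming a rigid upper-torso frame.
# The inner necklace swings within limits from the master bone, with very little twist.
# Both collar bones hang off spine_05 with angular limits, the right allowing wider first-axis swing than the left.
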PhysicsAsset /BRCosmetics/Characters/Player/Male/Medium/Bodies/M_MED_PlasticFork/Meshes/M_MED_PlasticFork_NeckLace
bodies {
 clavicle_l [{
   shapeType=Capsule
   length=10.6289
   radius=8.9653
   shapeName=clavicle_l_capsule
   center=<5.8282 -2.4089 -3.4317>
   rotation=<rot 90 -87.106 -94.5184>
  }]
 clavicle_r [{
   shapeType=Capsule
   length=9.4654
   radius=10.1289
   shapeName=clavicle_l_capsule
   center=<-5.5997 5.501 3.4317>
   rotation=<rot 90 172.875 -164.6116>
  }]
 def_collar_l [{
   shapeType=Capsule
   length=2.3285
   radius=1.6296
   shapeName=def_collar_l_capsule
   center=<3.5998 9.5226 0.8404>
   rotation=<rot 5.0831 -67.6438 -85.2214>
  }]
 def_collar_r [{
   shapeType=Capsule
   length=3.7709
   radius=1.1708
   shapeName=def_collar_r_capsule
   center=<3.265 -9.2298 1.1749>
   rotation=<rot 0.6525 -114.4267 -81.0443>
  }]
 dyn_necklace_MSTR [{shapeType=Capsule, length=0.5, radius=0.5}]
 dyn_necklace_inner [{shapeType=Capsule, length=3.8299, radius=3.4201, center=<3 0 2.5>, rotation=<rot 90 -0.0198 -0.0198>}]
 spine_04 [{shapeType=Capsule, length=15.2009, radius=12.4611, center=<3.5142 0.1594 0.0001>, rotation=<rot 0 89.9999 0>}]
 spine_05 [{shapeType=Capsule, length=19.8792, radius=11.1578, center=<-1.7112 -1.5237 0.0001>, rotation=<rot 0 89.9999 0>}]
}
constraints [
 {
  bone1Name=spine_05
  bone2Name=spine_04
  swing1LimitDegrees=45
  swing1Motion=Locked
  swing2LimitDegrees=45
  swing2Motion=Locked
  twistLimitDegrees=45
  twistMotion=Locked
 }
 {
  bone1Name=clavicle_l
  bone2Name=spine_05
  swing1LimitDegrees=45
  swing1Motion=Locked
  swing2LimitDegrees=45
  swing2Motion=Locked
  twistLimitDegrees=45
  twistMotion=Locked
 }
 {
  bone1Name=clavicle_r
  bone2Name=spine_05
  swing1LimitDegrees=45
  swing1Motion=Locked
  swing2LimitDegrees=45
  swing2Motion=Locked
  twistLimitDegrees=45
  twistMotion=Locked
 }
 {
  bone1Name=dyn_necklace_MSTR
  bone2Name=spine_05
  swing1LimitDegrees=45
  swing1Motion=Locked
  swing2LimitDegrees=45
  swing2Motion=Locked
  twistLimitDegrees=45
  twistMotion=Locked
 }
 {bone1Name=dyn_necklace_inner, bone2Name=dyn_necklace_MSTR, swing1LimitDegrees=10, swing2LimitDegrees=35, twistLimitDegrees=2}
 {bone1Name=def_collar_l, bone2Name=spine_05, swing1LimitDegrees=20, swing2LimitDegrees=20, twistLimitDegrees=20}
 {bone1Name=def_collar_r, bone2Name=spine_05, swing1LimitDegrees=40, swing2LimitDegrees=20, twistLimitDegrees=20}
]
