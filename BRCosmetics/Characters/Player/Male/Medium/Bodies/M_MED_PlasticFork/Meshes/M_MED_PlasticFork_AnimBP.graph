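# Cosmetic AnimBP for the PlasticFork male body that adds jacket, collar, necklace and lock-bone secondary motion to a pose copied from the attached parent mesh.
# Jacket correctives feed gameplay or front-end RigidBody sims, or bone offsets when rigid bodies are disabled.
# Collar offsets, AnimDynamics, a gravity-overridden necklace sim and deform Control Rig follow.
# The EventGraph infers crouch, leg, arm and pistol-grip alphas from bone transforms.
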AnimBlueprint /BRCosmetics/Characters/Player/Male/Medium/Bodies/M_MED_PlasticFork/Meshes/M_MED_PlasticFork_AnimBP
target_skeleton=&/M_MED_PlasticFork_Skeleton
graphs [
 AnimGraph {nodes [
   n1 = RigidBody!(ComponentPose=Local To Component(LocalPose=Use cached pose 'Pri sim'()), Alpha, OverrideWorldGravity:Vector="0,0,0", ExternalForce) -> (Pose) {
    bClampLinearTranslationLimitToRefPose=true
    bOverrideWorldGravity=true
    overridePhysicsAsset=&/M_MED_PlasticFork_Jacket
    simulationSpace=BaseBoneSpace
    alphaBoolBlend {blendOption=Linear}
    baseBoneRef {boneName=spine_02}
    componentAppliedLinearAccClamp=<3000 3000 3000>
    componentLinearAccScale {x=0.4}
    componentLinearVelScale {x=0.4}
    overrideWorldGravity {z=-980}
    simSpaceSettings {
     maxAngularAcceleration=200
     maxAngularVelocity=5
     maxLinearAcceleration=1000
     maxLinearVelocity=200
     velocityScaleZ=0
     worldAlpha=1
     externalLinearDragV=<1 1>
    }
   }
   n2 = Component To Local(ComponentPose=@n1)
   n3 = RigidBody(ComponentPose=Local To Component(LocalPose=Use cached pose 'Pri sim'()), OverrideWorldGravity="0,0,0") {
    bClampLinearTranslationLimitToRefPose=true
    bOverrideWorldGravity=true
    overridePhysicsAsset=&/M_MED_PlasticFork_Jacket_FrontEnd
    simulationSpace=BaseBoneSpace
    alphaBoolBlend {blendOption=Linear}
    baseBoneRef {boneName=spine_02}
    componentAppliedLinearAccClamp=<3000 3000 3000>
    componentLinearAccScale {x=0.4}
    componentLinearVelScale {x=0.4}
    overrideWorldGravity {z=-980}
    simSpaceSettings {
     maxAngularAcceleration=200
     maxAngularVelocity=5
     maxLinearAcceleration=1000
     maxLinearVelocity=200
     velocityScaleZ=0
     worldAlpha=1
     externalLinearDragV=<1 1>
    }
   }
   n4 = Component To Local(ComponentPose=@n3)
   n5 = Get Is Standing in FrontEnd() -> (Is Standing in FrontEnd:bool)
   n6 = "Transform (Modify) Bone - Bone: dyn_jacket_l_fr_MSTR"(
    ComponentPose=Local To Component(LocalPose=Blend Poses by bool(BlendPose_0=@n4, BlendPose_1=@n2, bActiveValue=@n5) {blendType=Linear})
    Alpha=Get Crouching_Sit() -> (Crouching_Sit:double)
    Translation="0.692,-0.161,0.229"
    Rotation="-5.181,2.494,-0.227"
    Scale="1,1,1"
   ) {
    rotationMode=BMM_Additive
    rotationSpace=BCS_BoneSpace
    translationMode=BMM_Additive
    translationSpace=BCS_BoneSpace
    alphaBoolBlend {blendOption=Linear}
    boneToModify {boneName=dyn_jacket_l_fr_MSTR}
    rotation=<rot -5.181 2.494 -0.227>
    translation=<0.692 -0.161 0.229>
   }
   n7 = Component To Local(ComponentPose=@n6)
   n8 = Get thigh_up_l_ALT() -> (thigh_up_l_ALT:double)
   n9 = "Transform (Modify) Bone - Bone: dyn_jacket_l_fr_MSTR"(
    ComponentPose=Local To Component(LocalPose=Use cached pose 'Pri sim'())
    Alpha=@n8
    Translation="5.280,-1.053,1.086"
    Rotation="16.996,-69.015,-114.922"
    Scale="1,1,1"
   ) {
    rotationMode=BMM_Additive
    rotationSpace=BCS_BoneSpace
    translationMode=BMM_Additive
    translationSpace=BCS_BoneSpace
    alphaBoolBlend {blendOption=Linear}
    boneToModify {boneName=dyn_jacket_l_fr_MSTR}
    rotation=<rot 16.996 -69.015 -114.922>
    translation=<5.28 -1.053 1.086>
   }
   n10 = "Transform (Modify) Bone - Bone: dyn_jacket_l_MSTR"(ComponentPose=@n9, Alpha=@n8, Translation="-2.893,-0.164,0.425", Rotation="14.930,3.550,0.272", Scale="1,1,1") {
    rotationMode=BMM_Additive
    rotationSpace=BCS_BoneSpace
    translationMode=BMM_Additive
    translationSpace=BCS_BoneSpace
    alphaBoolBlend {blendOption=Linear}
    boneToModify {boneName=dyn_jacket_l_MSTR}
    rotation=<rot 14.93 3.55 0.272>
    translation=<-2.893 -0.164 0.425>
   }
   n11 = Get thigh_up_r_ALT() -> (thigh_up_r_ALT:double)
   n12 = "Transform (Modify) Bone - Bone: dyn_jacket_r_fr_MSTR"(ComponentPose=@n10, Alpha=@n11, Translation="0,0,0", Rotation="13.000,38.833,68.266", Scale="1,1,1") {
    rotationMode=BMM_Additive
    rotationSpace=BCS_BoneSpace
    translationMode=BMM_Additive
    translationSpace=BCS_BoneSpace
    alphaBoolBlend {blendOption=Linear}
    boneToModify {boneName=dyn_jacket_r_fr_MSTR}
    rotation=<rot 13 38.833 68.266>
   }
   n13 = "Transform (Modify) Bone - Bone: dyn_jacket_r_MSTR"(ComponentPose=@n12, Alpha=@n11, Translation="0,0,0", Rotation="10,0,0", Scale="1,1,1") {
    rotationMode=BMM_Additive
    rotationSpace=BCS_BoneSpace
    translationMode=BMM_Additive
    translationSpace=BCS_BoneSpace
    alphaBoolBlend {blendOption=Linear}
    boneToModify {boneName=dyn_jacket_r_MSTR}
    rotation {pitch=10}
   }
   n14 = Component To Local(ComponentPose=@n13)
   n15 = Get bRigidBodyEnabled()
   n16 = "Transform (Modify) Bone - Bone: def_collar_r_MSTR"(
    ComponentPose=Local To Component(LocalPose=Blend Poses by bool(BlendPose_0=@n7, BlendPose_1=@n14, bActiveValue=@n15) {blendType=Linear})
    Alpha=Get Pist_arm_rot_up_R() -> (Pist_arm_rot_up_R:double)
    Translation="0,0,0"
    Rotation="7.914,-25.700,-5.052"
    Scale="1,1,1"
   ) {
    rotationMode=BMM_Additive
    rotationSpace=BCS_BoneSpace
    scaleMode=BMM_Additive
    scaleSpace=BCS_BoneSpace
    translationMode=BMM_Additive
    translationSpace=BCS_BoneSpace
    alphaBoolBlend {blendOption=Linear}
    boneToModify {boneName=def_collar_r_MSTR}
    rotation=<rot 7.914 -25.7 -5.052>
   }
   n17 = "Transform (Modify) Bone - Bone: def_collar_l_MSTR"(
    ComponentPose=@n16
    Alpha=Get Pist_arm_rot_up_L() -> (Pist_arm_rot_up_L:double)
    Translation="-0.496,0.922,4.638"
    Rotation="11.895,22.764,32.862"
    Scale="0.85,1.000,1.000"
   ) {
    rotationMode=BMM_Additive
    rotationSpace=BCS_BoneSpace
    scaleMode=BMM_Additive
    scaleSpace=BCS_BoneSpace
    translationMode=BMM_Additive
    translationSpace=BCS_BoneSpace
    alphaBoolBlend {blendOption=Linear}
    boneToModify {boneName=def_collar_l_MSTR}
    rotation=<rot 11.895 22.764 32.862>
    scale {x=0.85}
    translation=<-0.496 0.922 4.638>
   }
   n18 = "Transform (Modify) Bone - Bone: def_collar_r_MSTR"(
    ComponentPose=@n17
    Alpha=Get Pist_Arm_Down_R() -> (Pist_Arm_Down_R:double)
    Translation="0,0,0"
    Rotation="7.914,-25.700,-5.052"
    Scale="1,1,1"
   ) {
    rotationMode=BMM_Additive
    rotationSpace=BCS_BoneSpace
    scaleMode=BMM_Additive
    scaleSpace=BCS_BoneSpace
    translationMode=BMM_Additive
    translationSpace=BCS_BoneSpace
    alphaBoolBlend {blendOption=Linear}
    boneToModify {boneName=def_collar_r_MSTR}
    rotation=<rot 7.914 -25.7 -5.052>
   }
   n19 = "Transform (Modify) Bone - Bone: def_collar_l_MSTR"(
    ComponentPose=@n18
    Alpha=Get Pist_Arm_Down_L() -> (Pist_Arm_Down_L:double)
    Translation="-0.986,-0.111,2.618"
    Rotation="20.418,15.674,26.745"
    Scale="0.85,1.000,1.000"
   ) {
    rotationMode=BMM_Additive
    rotationSpace=BCS_BoneSpace
    scaleMode=BMM_Additive
    scaleSpace=BCS_BoneSpace
    translationMode=BMM_Additive
    translationSpace=BCS_BoneSpace
    alphaBoolBlend {blendOption=Linear}
    boneToModify {boneName=def_collar_l_MSTR}
    rotation=<rot 20.418 15.674 26.745>
    scale {x=0.85}
    translation=<-0.986 -0.111 2.618>
   }
   n20 = Anim Dynamics - Bone: dyn_lock_L(ComponentPose=@n19, Alpha=0.5, ExternalForce="0,0,0") {
    angularDampingOverride=1
    angularSpringConstant=500
    bAngularSpring=true
    bOverrideAngularDamping=true
    gravityScale=0.1
    numSolverIterationsPreUpdate=3
    alphaBoolBlend {blendOption=Linear}
    boundBone {boneName=dyn_lock_L}
    componentAppliedLinearAccClamp=<10000 10000 10000>
    physicsBodyDefinitions [{
      boundBone {boneName=dyn_lock_L}
      boxExtents=<11 11 11>
      constraintSetup {angularLimitsMax=<0 10 0>, angularLimitsMin=<0 -10 -180>, angularTarget=<0 0 0>}
      localJointOffset=<5 0 0>
     }]
   }
   n21 = Anim Dynamics - Bone: dyn_lock_R(ComponentPose=@n20, Alpha=0.5, ExternalForce="0,0,0") {
    angularDampingOverride=1
    angularSpringConstant=1500
    bAngularSpring=true
    bOverrideAngularDamping=true
    gravityScale=0.1
    numSolverIterationsPreUpdate=3
    alphaBoolBlend {blendOption=Linear}
    boundBone {boneName=dyn_lock_R}
    componentAppliedLinearAccClamp=<10000 10000 10000>
    physicsBodyDefinitions [{
      boundBone {boneName=dyn_lock_R}
      boxExtents=<11 11 11>
      constraintSetup {angularLimitsMax=<0 10 0>, angularLimitsMin=<0 -10 -180>, angularTarget=<0 0 0>}
      localJointOffset=<5 0 0>
     }]
   }
   n22 = Anim Dynamics - Bone: dyn_necklace_MSTR(ComponentPose=@n21, Alpha=1, ExternalForce="0,0,0") {
    angularDampingOverride=1
    angularSpringConstant=1500
    bAngularSpring=true
    bOverrideAngularDamping=true
    gravityScale=0.1
    numSolverIterationsPreUpdate=3
    alphaBoolBlend {blendOption=Linear}
    boundBone {boneName=dyn_necklace_MSTR}
    componentAppliedLinearAccClamp=<10000 10000 10000>
    physicsBodyDefinitions [{
      boundBone {boneName=dyn_necklace_MSTR}
      boxExtents=<11 11 11>
      constraintSetup {
       twistAxis=AxisZ
       angularLimitsMax=<5 0 0>
       angularLimitsMin=<0 -40 0>
       angularTarget=<0 0 0>
       linearAxesMax=<0.3 0 0>
       linearAxesMin=<-0.2 0 0>
      }
      localJointOffset=<5 0 0>
     }]
   }
   n23 = RigidBody(ComponentPose=@n22, Alpha=0.5, OverrideWorldGravity=Get Gravity Override()) {
    bClampLinearTranslationLimitToRefPose=true
    bOverrideWorldGravity=true
    overridePhysicsAsset=&/M_MED_PlasticFork_NeckLace
    simulationSpace=BaseBoneSpace
    alphaBoolBlend {blendOption=Linear}
    baseBoneRef {boneName=spine_05}
    componentAppliedLinearAccClamp=<2000 2000 1000>
    componentLinearAccScale {x=0.4}
    componentLinearVelScale {x=0.4}
    overrideWorldGravity {z=-980}
    simSpaceSettings {maxAngularAcceleration=100, maxAngularVelocity=5, maxLinearAcceleration=300, maxLinearVelocity=50, velocityScaleZ=0, worldAlpha=1}
   }
   n24 = Control Rig(Source=Component To Local(ComponentPose=@n23)) {
    alphaBoolBlend {blendOption=Linear}
    controlRigAssetReference {blueprintRigClass=&/M_MED_PlasticFork_Deform_ControlRig.M_MED_PlasticFork_Deform_ControlRig_C}
   }
   Output Pose(Result=@n24) {layerGroup=None}
   n25 = "Transform (Modify) Bone - Bone: dyn_necklace_MSTR"(
    ComponentPose=Local To Component(LocalPose=Copy Pose From Mesh() {bUseAttachedParent=true})
    Alpha=1
    Translation="0,0,1"
    Rotation="0,0,0"
    Scale="1,1,1"
   ) {
    rotationMode=BMM_Additive
    rotationSpace=BCS_BoneSpace
    translationMode=BMM_Additive
    translationSpace=BCS_BoneSpace
    alphaBoolBlend {blendOption=Linear}
    boneToModify {boneName=dyn_necklace_MSTR}
    translation {z=1}
   }
   n26 = Get thigh_up_l() -> (thigh_up_l:double)
   n27 = "Transform (Modify) Bone - Bone: dyn_jacket_l_fr_MSTR"(ComponentPose=@n25, Alpha=@n26, Translation="0,-3,3", Rotation="25,-35,-60", Scale="1,1,1") {
    rotationMode=BMM_Additive
    rotationSpace=BCS_BoneSpace
    translationMode=BMM_Additive
    translationSpace=BCS_BoneSpace
    alphaBoolBlend {blendOption=Linear}
    boneToModify {boneName=dyn_jacket_l_fr_MSTR}
    rotation=<rot 25 -35 -60>
    translation {y=-3, z=3}
   }
   n28 = "Transform (Modify) Bone - Bone: dyn_jacket_l_MSTR"(ComponentPose=@n27, Alpha=@n26, Translation="0,0,2", Rotation="15,0,0", Scale="1,1,1") {
    rotationMode=BMM_Additive
    rotationSpace=BCS_BoneSpace
    translationMode=BMM_Additive
    translationSpace=BCS_BoneSpace
    alphaBoolBlend {blendOption=Linear}
    boneToModify {boneName=dyn_jacket_l_MSTR}
    rotation {pitch=15}
    translation {z=2}
   }
   n29 = Get thigh_up_r() -> (thigh_up_r:double)
   n30 = "Transform (Modify) Bone - Bone: dyn_jacket_r_fr_MSTR"(ComponentPose=@n28, Alpha=@n29, Translation="0,3,3", Rotation="25,35,60", Scale="1,1,1") {
    rotationMode=BMM_Additive
    rotationSpace=BCS_BoneSpace
    translationMode=BMM_Additive
    translationSpace=BCS_BoneSpace
    alphaBoolBlend {blendOption=Linear}
    boneToModify {boneName=dyn_jacket_r_fr_MSTR}
    rotation=<rot 25 35 60>
    translation {y=3, z=3}
   }
   n31 = "Transform (Modify) Bone - Bone: dyn_jacket_r_MSTR"(ComponentPose=@n30, Alpha=@n29, Translation="0,0,0", Rotation="15,0,0", Scale="1,1,1") {
    rotationMode=BMM_Additive
    rotationSpace=BCS_BoneSpace
    translationMode=BMM_Additive
    translationSpace=BCS_BoneSpace
    alphaBoolBlend {blendOption=Linear}
    boneToModify {boneName=dyn_jacket_r_MSTR}
    rotation {pitch=15}
   }
   n32 = Get Crouching_Sit()
   n33 = "Transform (Modify) Bone - Bone: dyn_jacket_l_fr_MSTR"(ComponentPose=@n31, Alpha=@n32, Translation="0,0,0", Rotation="-23.852,21.481,-5.127", Scale="1,1,1") {
    rotationMode=BMM_Additive
    rotationSpace=BCS_BoneSpace
    translationMode=BMM_Additive
    translationSpace=BCS_BoneSpace
    alphaBoolBlend {blendOption=Linear}
    boneToModify {boneName=dyn_jacket_l_fr_MSTR}
    rotation=<rot -23.852 21.481 -5.127>
   }
   n34 = "Transform (Modify) Bone - Bone: dyn_jacket_l_bk_MSTR"(ComponentPose=@n33, Alpha=@n32, Translation="2,0,0", Rotation="46.144,-0.487,-26.870", Scale="1,1,1") {
    rotationMode=BMM_Additive
    rotationSpace=BCS_BoneSpace
    translationMode=BMM_Additive
    translationSpace=BCS_BoneSpace
    alphaBoolBlend {blendOption=Linear}
    boneToModify {boneName=dyn_jacket_l_bk_MSTR}
    rotation=<rot 46.144 -0.487 -26.87>
    translation {x=2}
   }
   n35 = "Transform (Modify) Bone - Bone: dyn_jacket_l_MSTR"(ComponentPose=@n34, Alpha=@n32, Translation="0,0,0", Rotation="7,0,0", Scale="1,1,1") {
    rotationMode=BMM_Additive
    rotationSpace=BCS_BoneSpace
    translationMode=BMM_Additive
    translationSpace=BCS_BoneSpace
    alphaBoolBlend {blendOption=Linear}
    boneToModify {boneName=dyn_jacket_l_MSTR}
    rotation {pitch=7}
   }
   Save cached pose 'Pri sim'(Pose=Component To Local(ComponentPose=@n35)) {cacheName=Pri sim}
  ]}
 EventGraph {nodes [
   Event BlueprintUpdateAnimation()
   > n1 = Sequence!(execute) -> (then_0, then_1, then_2:Exec, then_3:Exec)
   n2 = GetOwnerSkeletalMeshComponent()
   n3 = GetOwnerSkeletalMeshComponent()
   n4 = GetOwnerSkeletalMeshComponent()
   n5 = GetOwnerSkeletalMeshComponent()
   > Set Pistol(Pistol:double=float * float!(
     A=GetAlphaFromReferenceTransformDelta(
      SkeletalMeshComponent=@n3
      BoneName=index_metacarpal_r
      BaseName=index_03_r
      DeltaComponentType=EulerY
      InputRange=(X=-5,Y=0)
      TwistAxis="0,0,1"
     )
     B=GetAlphaFromReferenceTransformDelta(
      SkeletalMeshComponent=@n5
      BoneName=ring_metacarpal_r
      BaseName=ring_03_r
      DeltaComponentType=EulerY
      InputRange=(X=-5,Y=-4.9)
      TwistAxis="0,0,1"
     )
     C:float=GetAlphaFromReferenceTransformDelta(
      SkeletalMeshComponent=@n2
      BoneName=pinky_metacarpal_r
      BaseName=pinky_03_r
      DeltaComponentType=EulerY
      InputRange=(X=-6,Y=-5.9)
      TwistAxis="0,0,1"
     )
     D:float=GetAlphaFromReferenceTransformDelta(
      SkeletalMeshComponent=@n4
      BoneName=middle_metacarpal_r
      BaseName=middle_03_r
      DeltaComponentType=EulerY
      InputRange=(X=-7,Y=-5.9)
      TwistAxis="0,0,1"
     )
    ) -> (ReturnValue) {is=title K2Node_PromotableOperator}) -> (Output_Get:double)
   > n6 = PlayerMovement(AnimInstance=Self-Reference(), XDotProductCoefficient=1, YDotProductCoefficient=1, ZDotProductCoefficient=1)
   n7 = OR Boolean(A="float > float"(A=@n6, B=0.1), B="float < float"(A=@n6, B=-0.1))
   n8 = OR Boolean(A="float > float"(A=@n6.YMovement, B=0.1), B="float < float"(A=@n6.YMovement, B=-0.1))
   > Set moving(moving:double="To Float (Boolean)"(InBool=OR Boolean(A=@n7, B=@n8))) -> (Output_Get:double)
   n9 = GetOwnerSkeletalMeshComponent()
   n10 = BreakVector(InVec=BreakTransform(InTransform=GetDeltaTransformFromRefPose(self=@n9, BoneName=pelvis, BaseName=root)))
   n11 = MapRangeClamped(Value=@n10, InRangeA=-12, InRangeB=-14)
   > n12 = Set Crounching(Crounching:double=float * float!(A=@n11, B:bool=Get bIsCrouching()) -> (ReturnValue)) -> (Output_Get:double)
   > "Set Is Croching "("Is Croching ":bool="float > float"(A=@n12, B=0.5)) -> (Output_Get:bool)
   n13 = MapRangeClamped(Value=@n10, InRangeA=-35, InRangeB=-40)
   > Set Crouching_Moving(Crouching_Moving:double=float * float(A=MapRangeClamped(Value=@n13, InRangeA=0, InRangeB=1, OutRangeA=1, OutRangeB=0), B:double=@n11)) -> (Output_Get:double)
   > n14 = Branch(Condition="float <= float"(A=MapRangeClamped(Value=@n10, InRangeA=-55, InRangeB=-60)) -> (ReturnValue=~))
   > Set Crouching_Sit(Crouching_Sit:double=@n13) -> (Output_Get:double)
   Set Crouching_Sit(execute=@n14.else, Crouching_Sit=0)
   n15 = "To Float (Boolean)"(InBool=Get bRigidBodyEnabled())
   n16 = Get Crounching() -> (Crounching:double)
   n17 = "To Float (Boolean)"(InBool="float >= float"(A=MapRangeClamped(Value=@n16, InRangeA=0, InRangeB=1, OutRangeA=1, OutRangeB=0), B=0.5))
   n18 = GetOwnerSkeletalMeshComponent()
   n19 = BreakRotator(InRot=BreakTransform(InTransform=GetDeltaTransformFromRefPose(self=@n18, BoneName=thigh_l, BaseName=pelvis)).Rotation)
   n20 = float * float(A=@n17, B=MapRangeClamped(Value=Get moving() -> (moving:double), InRangeA=0, InRangeB=1, OutRangeA=1, OutRangeB=0))
   Set thigh_up_l(
    execute=@n1.then_2
    thigh_up_l:double=float * float!(A=@n17, B=MapRangeClamped(Value=@n19.Yaw, InRangeA=-20, InRangeB=-100), C:double=@n20, D:double=@n15) -> (ReturnValue)
   ) -> (Output_Get:double)
   n21 = GetOwnerSkeletalMeshComponent()
   n22 = BreakRotator(InRot=BreakTransform(InTransform=GetDeltaTransformFromRefPose(self=@n21, BoneName=thigh_r, BaseName=pelvis)).Rotation)
   > Set thigh_up_r(thigh_up_r:double=float * float(A=@n17, B=MapRangeClamped(Value=@n22.Yaw, InRangeA=-20, InRangeB=-100), C=@n20, D=@n15)) -> (Output_Get:double)
   n23 = BreakTransform(InTransform=GetDeltaTransformFromRefPose(self=GetOwnerSkeletalMeshComponent(), BoneName=upperarm_r, BaseName=spine_05))
   n24 = MapRangeClamped(Value="Absolute (Float)"(A=BreakRotator(InRot=@n23.Rotation).Yaw), InRangeA=20, InRangeB=100)
   n25 = Get Pistol() -> (Pistol:double)
   > Set Pist_arm_rot_up_R(Pist_arm_rot_up_R:double=FInterpTo(
     Current=Get Pist_arm_rot_up_R()
     Target=float * float!(A=@n25, B=@n24) -> (ReturnValue)
     DeltaTime=GetWorldDeltaSeconds!() -> (ReturnValue:double)
     InterpSpeed=10
    )) -> (Output_Get:double)
   n26 = BreakTransform(InTransform=GetDeltaTransformFromRefPose(self=GetOwnerSkeletalMeshComponent(), BoneName=upperarm_l, BaseName=spine_05))
   n27 = MapRangeClamped(Value="Absolute (Float)"(A=BreakRotator(InRot=@n26.Rotation).Yaw), InRangeA=20, InRangeB=120)
   n28 = Get Pistol()
   > Set Pist_arm_rot_up_L(Pist_arm_rot_up_L:double=FInterpTo(Current=Get Pist_arm_rot_up_L(), Target=float * float(A=@n27, B=@n28), DeltaTime=GetWorldDeltaSeconds(), InterpSpeed=10)) -> (Output_Get:double)
   > n29 = Branch(Condition=Get bIsInFrontEnd())
   n30 = BreakTransform(InTransform=GetDeltaTransformFromRefPose(self=GetOwnerSkeletalMeshComponent(), BoneName=upperarm_r, BaseName=spine_05))
   n31 = MapRangeClamped(Value="Absolute (Float)"(A=BreakRotator(InRot=@n30.Rotation).Yaw), InRangeA=130, InRangeB=20)
   > Set Pist_Arm_Down_R(Pist_Arm_Down_R:double=FInterpTo(Current=Get Pist_Arm_Down_R(), Target=@n31, DeltaTime=GetWorldDeltaSeconds(), InterpSpeed=10)) -> (Output_Get:double)
   n32 = BreakTransform(InTransform=GetDeltaTransformFromRefPose(self=GetOwnerSkeletalMeshComponent(), BoneName=upperarm_l, BaseName=spine_05))
   n33 = MapRangeClamped(Value="Absolute (Float)"(A=BreakRotator(InRot=@n32.Rotation).Yaw), InRangeA=120, InRangeB=20)
   > Set Pist_Arm_Down_L(Pist_Arm_Down_L:double=FInterpTo(Current=Get Pist_Arm_Down_L(), Target=@n33, DeltaTime=GetWorldDeltaSeconds(), InterpSpeed=10)) -> (Output_Get:double)
   Set Pist_Arm_Down_R(execute=@n29.else, Pist_Arm_Down_R=FInterpTo(Current=Get Pist_Arm_Down_R(), Target=@n31, DeltaTime=GetWorldDeltaSeconds(), InterpSpeed=10))
   n34 = Get Pistol()
   > Set Pist_Arm_Down_L(Pist_Arm_Down_L=FInterpTo(Current=Get Pist_Arm_Down_L(), Target=float * float(A=@n33, B=@n34), DeltaTime=GetWorldDeltaSeconds(), InterpSpeed=10))
   n35 = Branch(execute=@n1.then_3, Condition="float >= float"(A=Get Crouching_Sit(), B=0.5))
   n36 = MapRangeClamped(Value="To Float (Boolean)"(InBool=Get bRigidBodyEnabled()), InRangeA=0, InRangeB=1, OutRangeA=1, OutRangeB=0)
   > Set thigh_up_l_ALT(thigh_up_l_ALT:double=float * float(A=@n36, B=0.65)) -> (Output_Get:double)
   > Set thigh_up_r_ALT(thigh_up_r_ALT:double=float * float(A=@n36, B=0.5)) -> (Output_Get:double)
   n37 = Branch(execute=@n35.else, Condition="Get Is Croching "() -> ("Is Croching ":bool))
   n38 = GetOwnerSkeletalMeshComponent()
   n39 = BreakRotator(InRot=BreakTransform(InTransform=GetDeltaTransformFromRefPose(self=@n38, BoneName=thigh_l, BaseName=pelvis)).Rotation)
   > Set thigh_up_l_ALT(thigh_up_l_ALT=MapRangeClamped(Value=@n39.Yaw, InRangeA=0, InRangeB=-180))
   n40 = GetOwnerSkeletalMeshComponent()
   n41 = BreakRotator(InRot=BreakTransform(InTransform=GetDeltaTransformFromRefPose(self=@n40, BoneName=thigh_r, BaseName=pelvis)).Rotation)
   > Set thigh_up_r_ALT(thigh_up_r_ALT=MapRangeClamped(Value=@n41.Yaw, InRangeA=0, InRangeB=-100))
   Set thigh_up_l_ALT(execute=@n37.else, thigh_up_l_ALT=float * float(A=MapRangeClamped(Value=@n39.Yaw, InRangeA=-10, InRangeB=-140), B=@n36))
   > Set thigh_up_r_ALT(thigh_up_r_ALT=float * float(A=@n36, B=MapRangeClamped(Value=@n41.Yaw, InRangeA=-20, InRangeB=-100)))
   n42 = Gravity Override State Function DEPRECATED(
    execute=@n1.then_1
    CharacterPartAnimInstance=Self-Reference()
    GameTimeInSeconds=GetGameTimeInSeconds()
    GravityOverrideParams=Get Gravity Override Params()
   )
   > Set Gravity Override(Gravity Override=@n42)
   n43 = Get bIsPlayingEmote()
   > Set Is Standing in FrontEnd(Is Standing in FrontEnd:bool=AND Boolean(A=Get bIsInFrontEnd(), B=NOT Boolean(A=@n43))) -> (Output_Get:bool)
   n44 = Get bIsInFrontEnd()
   float * float(A=@n31, B=float * float(A=Get Pistol(), B:bool=NOT Boolean(A=@n44)))
  ]}
]
variables [
 Check Fings {type=number}
 Crouching_Moving {type=number}
 Crouching_Sit {type=number}
 Crounching {type=number}
 Gravity Override {type=Vector}
 Gravity Override Params {
  type=GravityOverrideParamsStruct
  default {
   skydive_DivingDown {gravityOverride=<-1500 0 200>}
   skydive_FloatingDown {gravityOverride=<0 -700 -200>}
   skydive_GoingUp {gravityOverride=<0 -300 -1500>}
   skydive_Parachuting {gravityOverride=<0 -800 0>}
  }
 }
 "Is Croching " {type=boolean}
 Is Standing in FrontEnd {type=boolean}
 moving {type=number}
 Pist_Arm_Down_L {type=number}
 Pist_Arm_Down_R {type=number}
 Pist_arm_rot_up_L {type=number}
 Pist_arm_rot_up_R {type=number}
 Pistol {type=number}
 thigh_up_l {type=number}
 thigh_up_l_ALT {type=number}
 thigh_up_r {type=number}
 thigh_up_r_ALT {type=number}
]
