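# Animation Blueprint for the face accessory of the PlasticFork male character, adding physics-driven secondary motion around the head.
# It copies the parent mesh pose, layers in right-arm and weapon branches from a stored custom part mesh, runs a Control Rig and caches the result.
# Two head-space RigidBody simulations branch from that cache, blended by bIsTargeting to use a separate aiming physics asset.
# The event graph stores the part mesh and updates a skydive-aware gravity override variable.
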AnimBlueprint /BRCosmetics/Characters/Player/Male/Medium/Bodies/M_MED_PlasticFork/Meshes/Parts/M_MED_PlasticFork_FaceAcc_AnimBP
target_skeleton=&/M_MED_PlasticFork_FaceAcc_Skeleton
graphs [
 AnimGraph {nodes [
   n1 = RigidBody!(ComponentPose=Local To Component(LocalPose=Use cached pose 'Cach A'()), Alpha, OverrideWorldGravity:Vector="0,0,0", ExternalForce) -> (Pose) {
    bOverrideWorldGravity=true
    overridePhysicsAsset=&/M_MED_PlasticFork_FaceAcc_PhysicsAiming
    simulationSpace=BaseBoneSpace
    alphaBoolBlend {blendOption=Linear}
    baseBoneRef {boneName=head}
    componentAppliedLinearAccClamp=<2000 2000 2000>
    componentLinearAccScale=<0.2 0.2 0.2>
    componentLinearVelScale=<1 1 1>
    simSpaceSettings {maxAngularAcceleration=50, maxAngularVelocity=5, maxLinearAcceleration=50, maxLinearVelocity=5, velocityScaleZ=0, worldAlpha=1}
   }
   n2 = RigidBody(ComponentPose=Local To Component(LocalPose=Use cached pose 'Cach A'()), Alpha=0.7, OverrideWorldGravity="0,0,0") {
    bOverrideWorldGravity=true
    overridePhysicsAsset=&/M_MED_PlasticFork_FaceAcc_Physics
    simulationSpace=BaseBoneSpace
    alphaBoolBlend {blendOption=Linear}
    baseBoneRef {boneName=head}
    componentAppliedLinearAccClamp=<1500 1500 1500>
    componentLinearAccScale=<0.1 0.1 0.1>
    componentLinearVelScale=<1 1 1>
    simSpaceSettings {maxAngularAcceleration=50, maxAngularVelocity=5, maxLinearAcceleration=50, maxLinearVelocity=5, velocityScaleZ=0, worldAlpha=1}
   }
   Output Pose(Result=Blend Poses by bool(
     BlendPose_0=Component To Local(ComponentPose=@n1)
     BlendPose_1=Component To Local(ComponentPose=@n2)
     BlendTime_1=0.3
     bActiveValue=Get bIsTargeting()
    ) {blendType=Linear}) {layerGroup=None}
   n3 = Layered blend per bone(
    BasePose=Copy Pose From Mesh() {bUseAttachedParent=true}
    BlendPoses_0=Copy Pose From Mesh(SourceMeshComponent=Get Custom Mesh Part() -> (Custom Mesh Part:Skeletal Mesh Component Object Reference))
   ) {layerSetup [{branchFilters [{boneName=clavicle_r}, {boneName=upperarm_r}, {boneName=lowerarm_r}, {boneName=weapon_r}]}]}
   n4 = Control Rig(Source=@n3) {
    alphaBoolBlend {blendOption=Linear}
    controlRigAssetReference {blueprintRigClass=&/M_MED_PlasticFork_FaceAcc_CtrlRig.M_MED_PlasticFork_FaceAcc_CtrlRig_C}
   }
   Save cached pose 'Cach A'(Pose=@n4) {cacheName=Cach A}
  ]}
 EventGraph {nodes [
   Event BlueprintUpdateAnimation()
   > n1 = Gravity Override State Function DEPRECATED(CharacterPartAnimInstance=Self-Reference(), GameTimeInSeconds=GetGameTimeInSeconds(), GravityOverrideParams=Get Gravity Override Params())
   > Set Gravity Override(Gravity Override=@n1)
   n2 = Event OnCharacterCustomizationLoaded()
   > Set Custom Mesh Part(Custom Mesh Part:Skeletal Mesh Component Object Reference=Get Array Item with EFort Custom Part Type(Array=@n2.CustomMeshParts)) -> (Output_Get:Skeletal Mesh Component Object Reference)
   n3 = GetControlRotation!(self:Pawn Object Reference=TryGetPawnOwner()) -> (ReturnValue:Rotator)
   n4 = GetForwardVector!(InRot:Rotator=@n3) -> (ReturnValue:Vector)
   n5 = GetRightVector!(InRot:Rotator=@n3) -> (ReturnValue:Vector)
   n6 = GetUpVector!(InRot:Rotator=@n3) -> (ReturnValue:Vector)
   MapRangeClamped(
    Value=MapRangeClamped(
     Value=MakeRotationFromAxes!(Forward:Vector=@n4, Right:Vector=@n5, Up:Vector=@n6) -> (ReturnValue_Roll:float, ReturnValue_Pitch:float, ReturnValue_Yaw:float).ReturnValue_Pitch
     InRangeA=30
     InRangeB=60
    )
    InRangeA=0
    InRangeB=1
    OutRangeA=1
    OutRangeB=0
   )
   Get Gravity Override Params_Old() -> (Gravity Override Params_Old:Gravity Override Params Struct Structure)
  ]}
]
variables [
 Custom Mesh Part {type=SkeletalMeshComponent}
 Gravity Override {type=Vector, default=<0 0 -980>}
 Gravity Override Params {
  type=GravityOverrideParamsStruct
  default {
   skydive_DivingDown {gravityOverride=<500 0 0>, windAmplitude=<50 150 50>}
   skydive_FloatingDown {gravityOverride=<0 -300 -400>, windAmplitude=<50 -150 100>, windOffset=<0 -150 100>}
   skydive_GoingUp {gravityOverride=<0 0 0>, windAmplitude=<100 -250 200>}
  }
 }
 Gravity Override Params_Old {
  type=GravityOverrideParamsStruct
  default {
   skydive_DivingDown {gravityOverride=<1500 0 0>}
   skydive_FloatingDown {gravityOverride=<0 -700 0>}
   skydive_GoingUp {gravityOverride=<0 -300 200>}
  }
 }
 SwitchAimPhysOnHard {type=boolean}
]
